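; UI director component for the Snow Grenade that adds and removes the SnowUIScene through the owning Dynamic UI Director as a gameplay tag changes.
; When UI is ready, it starts tag watching on the local player controller and binds view target changes to watch each new target.
; After validity and mobile-touch checks, async ability tasks wait for the tag's addition or removal to drive the scene.
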(event EventOnReadyToShowUI
  (bind _asfort_player_controller (Utilities|Casting|CastToFortPlayerController (CallFunction|GetOwningLocalPlayerController)))
  (EventDispatchers|BindEventtoOnViewTargetChanged _asfort_player_controller (delegate Custom|OnViewTargetChanged))
  (CallFunction|WaitforGameplayTag _asfort_player_controller))

(event Custom|OnViewTargetChanged
  (params (PlayerController :type Fort%20Player%20Controller%20Object%20Reference) (OldViewTarget :type Actor%20Object%20Reference) (NewViewTarget :type Actor%20Object%20Reference))
  (CallFunction|WaitforGameplayTag NewViewTarget))

(event Custom|WaitForGameplayTag
  (params (PlayerActor :type Actor%20Object%20Reference))
  (bind _returnvalue (CallFunction|GetOwningLocalPlayerController))
  ;[1 1 1] Wait For Appropriate Tag
  (Utilities|IsValid PlayerActor
    (:Is%20Valid
      (Utilities|IsValid _returnvalue
        (:Is%20Valid
          (if (not (and (TouchPlatform|IsMobileGame) (TouchPlatform|IsUsingTouch _returnvalue)))
            (Utilities|Casting|CastToFortPlayerPawnAthena (Pawn|GetControlledPawn _returnvalue))
            (Ability|Tasks|WaitGameplayTagAddToActor PlayerActor (Variables|Default|GetSnowGrenade%20Tag%20Event)
              (:then
                (Ability|Tasks|WaitGameplayTagRemoveFromActor PlayerActor (Variables|Default|GetSnowGrenade%20Tag%20Event)
                  (:Removed
                    (DynamicUIDirector|RemoveScene (CallFunction|GetOwningFortDynamicUIDirector) "/FirePetalSnowGrenade/UI/SnowUIScene.SnowUIScene"))))
              (:Added
                (DynamicUIDirector|AddScene (CallFunction|GetOwningFortDynamicUIDirector) "/FirePetalSnowGrenade/UI/SnowUIScene.SnowUIScene")))))))))
